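# Translucent unlit, two-sided material for Niagara mesh particles, named as a spline ring effect for Time Trial gameplay.
# SplineThicken sets a randomized per-particle width, and panning curl vector noise adds wobble to its world position offset.
# Particle color feeds emissive adjustments, while two gradient ramps build a depth-faded opacity, optionally shaped by Fresnel and multiplied by particle alpha.
# Dynamic parameters drive width, curl intensity and Fresnel exponent from the particle system.
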Material /Game/Effects/Fort_Effects/Gameplay/TimeTrial/M_MythicLabors_Spline_Ring
graph {
 properties {
  bEnableResponsiveAA=true
  bUsedWithNiagaraMeshParticles=true
  blendMode=BLEND_Translucent
  pixelDepthOffsetMode=PDOM_Legacy
  shadingModel=MSM_Unlit
  twoSided=true
  usedShadingModels=Unlit
 }
 nodes [
  n1 = Multiply(
   A=DynamicParameter() {defaultValue=<rgba 50 1 1 1>, paramNames [SplineWidth,Curlntensity,FresnelExp,Param4]}.SplineWidth
   B=LinearInterpolate(Alpha=ParticleRandom()) {constA=0.7, constB=1.3}
  )
  n2 = MaterialFunctionCall(WidthBase=@n1, WidthTip=@n1) {materialFunction=&SplineThicken}
  n3 = Add(A=Subtract(A=WorldPosition().XYZ, B=ParticlePositionWS()), B=ParticleRandom())
  n4 = Multiply(A=Time(), B=VectorParameter() {parameterName=CurlPan, defaultValue=<rgba 3 0 0 0>}.RGB)
  n5 = Divide(
   A=Transform(None=Add(A=@n3, B=@n4)) {transformSourceType=TRANSFORMSOURCE_World, transformType=TRANSFORM_Local}
   B=VectorParameter() {parameterName=CurlTileScale, defaultValue=<rgba 200 200 200 0>}.RGB
  )
  n6 = VectorParameter() {parameterName=CurlIntensity, defaultValue=<rgba 1 1 1 0>}
  n7 = DynamicParameter() {defaultValue=<rgba 50 1 1 1>, paramNames [SplineWidth,Curlntensity,FresnelExp,Param4]}
  Add(
   A=@n2.WorldPositionOffset
   B=Multiply(A=VectorNoise(World Position=@n5) {noiseFunction=VNF_CurlALU}, B=Multiply(A=@n6.RGB, B=@n7.Curlntensity))
  ) {constB=0}
  n8 = ParticleColor()
  MaterialFunctionCall(Emissive=@n8.RGB) {materialFunction=&MF_FX_EmissiveAdjustments}
  n9 = TextureSample(UVs=MaterialFunctionCall(UVs=TextureCoordinate(), Texture Scale=Constant2Vector() {g=1, r=0.25}) {materialFunction=&ScaleUVsByCenter}.UVs) {samplerSource=SSM_Clamp_WorldGroupSettings, texture=/Game/Global/Textures/Default/Masks_and_Gradients/T_Ramp_Virt}
  n10 = Constant() {r=0.1}
  n11 = TextureSample(UVs=MaterialFunctionCall(UVs=TextureCoordinate(), "Rotation Angle (0-1)"=Constant() {r=0.25}) {materialFunction=&CustomRotator}.Rotated Values) {samplerSource=SSM_Clamp_WorldGroupSettings, texture=/Game/Global/Textures/Default/Masks_and_Gradients/T_Water_Top_Ramp}
  n12 = Constant() {r=0.25}
  n13 = DepthFade(
   Opacity=Add(
    A=MaterialFunctionCall(In=@n9.RGB, Contrast=@n10) {materialFunction=&CheapContrast}.Result
    B=Multiply(A=@n11.RGB, B=@n12) {constB=0.25}
   )
   FadeDistance=ScalarParameter() {defaultValue=1, parameterName=DepthFadeDistance}
  )
  n14 = MaterialFunctionCall(
   Normal Vector=@n2.Normal
   Invert Fresnel=StaticBoolParameter() {defaultValue=true, parameterName=InvertFresnel}
   Power=DynamicParameter() {defaultValue=<rgba 50 1 1 1>, paramNames [SplineWidth,Curlntensity,FresnelExp,Param4]}.FresnelExp
  ) {materialFunction=&Fresnel_Function}
  Multiply(
   A=@n8.A
   B=LinearInterpolate(A=@n13, B=Multiply(A=@n13, B=@n14.Result), Alpha=ScalarParameter() {defaultValue=1, parameterName=UseFresnel}) {constA=1}
  ) {constB=0.25}
  ScalarParameter() {defaultValue=1, parameterName=FresnelExponent}
 ]
}
