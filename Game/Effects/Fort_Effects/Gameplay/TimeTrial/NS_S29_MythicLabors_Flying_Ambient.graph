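# Ambient Niagara effect for the Season 29 Mythic Labors flying time trial, forming a glowing orange-gold ring of sparks.
# Bits spawns two masked groups on cylinder and torus shapes of equal radius, with curl noise, gravity, drag and pulsing size and alpha.
# Glow_Light holds one persistent particle driving a light and two sprites.
# Sprite brightness rises with camera distance, while light intensity follows time of day and User.FXBrightnessScale.
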NiagaraSystem /Game/Effects/Fort_Effects/Gameplay/TimeTrial/NS_S29_MythicLabors_Flying_Ambient
dependencies {
 usedDataInterfaces [&NiagaraDataInterfaceCurve, &NiagaraDataInterfaceColorCurve, &NiagaraDataInterfaceVector2DCurve]
 usedDynamicInputs [&HueShiftLinearColor]
 usedModules [&SystemState]
}
emitters {
 Bits {
  emitterUpdateScript [
   &EmitterState(Enable Distance Culling=true, Module.bUseMaxDistance=true, Module.bUseMinDistance=true, Scalability Mode=::NewEnumerator1)
   &SpawnRate(SpawnRate=175, Use Spawn Probability=false)
   &SpawnRate(SpawnGroup=1, SpawnRate=75, Use Spawn Probability=false) {module=SpawnRate001}
  ]
  particleSpawnScript [
   &InitializeParticle(
    Color Maximum=<rgba 1 0.9355 0 1>
    Color Minimum=<rgba 1 0.2226 0 1>
    Color Mode=::NewEnumerator2
    Lifetime=&Lerp_Float(
     A=0.6
     B=3
     Alpha=&FloatFromCurve(
      FloatCurve=DataInterface() {propertyValues {Curve {keys [
          {interpMode=RCIM_Cubic, tangentMode=RCTM_User, time=0, value=0}
          {arriveTangent=0.9713, interpMode=RCIM_Cubic, leaveTangent=0.9713, tangentMode=RCTM_User, time=0.8, value=0.25}
          {arriveTangent=5.7308, interpMode=RCIM_Cubic, leaveTangent=5.7308, tangentMode=RCTM_User, time=1, value=1}
         ]}}}
      CurveIndex=&RandomRangeFloat(Minimum=0, Maximum=1)
     )
    )
    Sprite Size Mode=::NewEnumerator1
    Uniform Sprite Size Max=3
    Uniform Sprite Size Min=1
    UsePositionOffset=false
    Write Material Random=true
   )
   &ShapeLocation(
    Cylinder Height=20
    Cylinder Radius=236
    MaskSpawnGroup=true
    Override Seed=false
    Rotation Mode=::NewEnumerator1
    Shape Primitive=::NewEnumerator1
    Surface Only=true
    Surface Only Band Thickness=0.1
    Use Endcaps In Surface Only Mode=false
    UseOwnerScale=false
   )
   &AddVelocity(
    Cone Angle=10
    Module.UseConeFalloff=false
    Module.UseVelDistribution=false
    Override Seed=false
    Velocity Mode=::NewEnumerator2
    Velocity Speed=&Lerp_Float(
     A=15
     B=50
     Alpha=&FloatFromCurve(
      FloatCurve=DataInterface() {propertyValues {Curve {keys [
          {interpMode=RCIM_Cubic, tangentMode=RCTM_User, time=0, value=0}
          {arriveTangent=1.2065, interpMode=RCIM_Cubic, leaveTangent=1.2065, tangentMode=RCTM_User, time=0.7, value=0.3}
          {arriveTangentWeight=1.0116, interpMode=RCIM_Cubic, tangentMode=RCTM_User, time=1, value=1}
         ]}}}
      CurveIndex=&RandomRangeFloat(Minimum=-1, Maximum=1)
      Scale Curve=&RandomRangeFloat(Minimum=-1, Maximum=1)
     )
    )
   ) {enabled=false}
   &ShapeLocation(
    Handle Radius=20
    Large Radius=236
    MaskSpawnGroup=true
    Override Seed=false
    Rotation Angle=90
    Rotation Angle Type:ENiagara_AngleInput=::NewEnumerator0
    Rotation Mode=::NewEnumerator1
    Shape Primitive=::NewEnumerator3
    Spawn Group Mask=1
    U Distribution=0.5
    UseOwnerScale=false
   ) {module=ShapeLocation001}
   &AddVelocity(Velocity=&MaskVectorBySpawnGroup(
     Masked Vector=&Multiply_VectorByFloat(Vector=<0 0 -1>, Float=&RandomRangeFloat(Minimum=-5, Maximum=100))
     Passthrough Vector=<0 0 0>
     Spawn Group Mask=1
    )) {module=AddVelocity002}
  ]
  particleUpdateScript [
   &ParticleState()
   &CurlNoiseForce(
    Noise Frequency=25
    Noise Strength=&Multiply_Float(
     A=150
     B=&FloatFromCurve(FloatCurve=DataInterface() {propertyValues {Curve {keys [{interpMode=RCIM_Cubic, tangentMode=RCTM_User, time=0, value=1}, {interpMode=RCIM_Cubic, tangentMode=RCTM_User, time=1, value=0}]}}})
    )
    Pan Noise=false
   )
   &GravityForce(Gravity=&RandomRangeVector(
     Minimum=<0 0 0>
     Maximum=&Multiply_VectorByFloat(
      Vector=<0 0 -1>
      Float=&FloatFromCurve(FloatCurve=DataInterface() {propertyValues {Curve {keys [<at 0 1>,<at 1 0>]}}}, Scale Curve=500)
     )
    ))
   &Drag(
    Drag=&FloatFromCurve(
     FloatCurve=DataInterface() {propertyValues {Curve {keys [{interpMode=RCIM_Cubic, tangentMode=RCTM_Break, time=0, value=1}, {interpMode=RCIM_Cubic, tangentMode=RCTM_Break, time=1, value=0.2}]}}}
     Scale Curve=5
    )
    Use Linear Drag=true
    Use Rotational Drag=false
   )
   &SolveForcesAndVelocity(Clamp Velocity=false, Limit Acceleration=false)
   &ScaleColor(
    Scale Alpha=&FloatFromCurve(
     FloatCurve=DataInterface() {propertyValues {Curve {keys [
         {interpMode=RCIM_Cubic, tangentMode=RCTM_User, time=0, value=0}
         {interpMode=RCIM_Cubic, tangentMode=RCTM_User, time=0.15, value=1}
         {interpMode=RCIM_Cubic, tangentMode=RCTM_User, time=0.3, value=2.875}
         {interpMode=RCIM_Cubic, tangentMode=RCTM_User, time=0.4, value=1.25}
         {interpMode=RCIM_Cubic, tangentMode=RCTM_User, time=0.5, value=1.2}
         {interpMode=RCIM_Cubic, tangentMode=RCTM_User, time=0.6, value=2.125}
         {interpMode=RCIM_Cubic, tangentMode=RCTM_User, time=0.7, value=1.2}
         {interpMode=RCIM_Cubic, tangentMode=RCTM_User, time=0.8, value=1}
         {interpMode=RCIM_Cubic, tangentMode=RCTM_User, time=1, value=0}
        ]}}}
     Scale Curve=&FloatFromCurve(FloatCurve=DataInterface() {propertyValues {Curve {keys [
          {interpMode=RCIM_Cubic, tangentMode=RCTM_Break, time=0, value=0}
          {interpMode=RCIM_Cubic, tangentMode=RCTM_User, time=0.3, value=1}
          {interpMode=RCIM_Cubic, tangentMode=RCTM_Break, time=1, value=0}
         ]}}})
    )
    Scale RGB=&VectorFromFloat(Value=&Lerp_Float(
      A=1
      B=10
      Alpha=&FloatFromCurve(
       FloatCurve=DataInterface() {propertyValues {Curve {keys [
           {interpMode=RCIM_Cubic, tangentMode=RCTM_Break, time=0.1, value=0}
           {arriveTangent=3.4325, interpMode=RCIM_Cubic, tangentMode=RCTM_Break, time=1, value=1}
          ]}}}
       CurveIndex=&RandomRangeFloat(Minimum=0, Maximum=1)
      )
     ))
    ScaleA=true
    ScaleRGB=true
   )
   &ScaleSpriteSize(Uniform Curve Sprite Scale=DataInterface() {propertyValues {Curve {keys [
        {interpMode=RCIM_Cubic, tangentMode=RCTM_User, time=0, value=1}
        {interpMode=RCIM_Cubic, tangentMode=RCTM_User, time=0.15, value=1}
        {interpMode=RCIM_Cubic, tangentMode=RCTM_User, time=0.3, value=2}
        {interpMode=RCIM_Cubic, tangentMode=RCTM_User, time=0.4, value=1.25}
        {interpMode=RCIM_Cubic, tangentMode=RCTM_User, time=0.5, value=1.2}
        {interpMode=RCIM_Cubic, tangentMode=RCTM_User, time=0.6, value=2.125}
        {interpMode=RCIM_Cubic, tangentMode=RCTM_User, time=0.8, value=1.5}
        {interpMode=RCIM_Cubic, tangentMode=RCTM_User, time=0.9, value=1.375}
        {interpMode=RCIM_Cubic, tangentMode=RCTM_User, time=1, value=0}
       ]}}})
  ]
  rendererClasses [&NiagaraSpriteRendererProperties]
  renderers [{rendererClass=&NiagaraSpriteRendererProperties}]
 }
 Glow_Light {
  emitterUpdateScript [
   &EmitterState(Life Cycle Mode=::NewEnumerator1, Loop Behavior=::NewEnumerator1, Loop Duration Mode=::NewEnumerator1, UseLoopDelay=false)
   &SpawnBurst_Instantaneous(Use Spawn Probability=false, UseLoopCountLimit=false)
  ]
  particleSpawnScript [
   &InitializeParticle(Color=<rgba 2.5 0.9979 0 1>, Sprite Size Mode=::NewEnumerator3, Uniform Sprite Size=600, UsePositionOffset=false, Write Material Random=true)
   &SpriteFacingAndAlignment(Alignment Coordinate Space=::Local, Facing Coordinate Space=::Local, Sprite Alignment=<1 0 0>, WriteAlignment=true, WriteFacing=true)
  ]
  particleUpdateScript [
   /SetVariables(Particles.DarkerCameraOffset=-5, Particles.Darker_Color:LinearColor=<rgba 0.5 0.1632 0 5>, Particles.Darker_Size:Vector2f=<415 415>) {isSetParameters=true}
   &ParticleState(Kill Particles When Lifetime Has Elapsed=false, Let Infinitely Lived Particles Die When Emitter Deactivates=true)
   &ScaleSpriteSize(Scale Sprite Size Mode=::NewEnumerator0)
   &SolveForcesAndVelocity(Clamp Velocity=false, Limit Acceleration=false)
   &CameraOffset(Camera Offset Amount=10)
   &FN_Light_Attributes(
    Color=<rgba 1 0.5487 0 1>
    Create Color Attribute=true
    Create Exponent Attribute=true
    Create Position Attribute=true
    Create Radius Attribute=true
    Position=&Add_Vector(
     A=&ConvertPositionToVector(Input Position=$Particles.Position, Convert To Absolute Large World Vector=false)
     B=<0 0 15>
    )
    Radius=&Multiply_Float(A=600, B=1)
    Scale Color=&Multiply_Float(
     A=&FNTimeOfDayScalar(Night Scale=0.75, Morning Scale=1, Evening Scale=0.75)
     B=&Multiply_Float(A=3, B=$User.FXBrightnessScale)
    )
   )
   &ScaleColor(
    Scale RGB=&VectorFromFloat(Value=&Multiply_Float(
      A=1
      B=&FloatFromCurve(
       FloatCurve=DataInterface() {propertyValues {Curve {keys [{interpMode=RCIM_Cubic, tangentMode=RCTM_User, time=0, value=0.1}, {interpMode=RCIM_Cubic, tangentMode=RCTM_User, time=0.5, value=1}]}}}
       CurveIndex=&FN_LOD_Distance_Remap(Start Distance=200, End Distance=20000, Minimum Output=0, Maximum Output=1, Clamp Output=true)
       Scale Curve=10
      )
     ))
    ScaleA=true
    ScaleRGB=true
   )
   &DynamicMaterialParameters(
    Index 0 Param 1=0.5
    Index 0 Param 2=0.4
    Index 0 Param 3=1
    Index 0 Param 4=1
    Module.Param0WriteEnabled=true
    Module.Param1WriteEnabled=true
    Module.Param2WriteEnabled=true
    Module.Param3WriteEnabled=true
   )
  ]
  rendererClasses [&NiagaraLightRendererProperties, &NiagaraSpriteRendererProperties]
  renderers [
   {rendererClass=&NiagaraLightRendererProperties, rendererIndex=0}
   {rendererClass=&NiagaraSpriteRendererProperties, rendererIndex=1}
   {rendererClass=&NiagaraSpriteRendererProperties, rendererIndex=2}
  ]
 }
}
summary {
 systemName=NS_S29_MythicLabors_Flying_Ambient
 emitters [
  {emitterName=Bits, rendererClasses [&NiagaraSpriteRendererProperties]}
  {emitterName=Glow_Light, rendererClasses [&NiagaraLightRendererProperties, &NiagaraSpriteRendererProperties]}
 ]
 userVariables [
  User.FXBrightnessScale {type=NiagaraFloat, defaultValue=1, description=""}
  User.FXDistance {type=NiagaraFloat, defaultValue=8000, description=""}
 ]
}
